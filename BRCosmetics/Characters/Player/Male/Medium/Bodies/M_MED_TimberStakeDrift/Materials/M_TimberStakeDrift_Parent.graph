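# Parent material for the TimberStakeDrift male character body, printing a configurable number of up to three digits onto the base color.
# The clamped Number is split with Fmod to drive three FlipBook lookups into a digit atlas, placed side by side on UV1.
# The masked digits blend DigitsColor over the diffuse from MF_CharacterParent_2023_PreFX, and the EnableDigits static switch feeds the result to MF_CharacterParent_2023_PostFX.
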Material /BRCosmetics/Characters/Player/Male/Medium/Bodies/M_MED_TimberStakeDrift/Materials/M_TimberStakeDrift_Parent
graph {
 properties {
  bUseMaterialAttributes=true
  bUsedWithMeshDeformer=true
  bUsedWithMorphTargets=true
  bUsedWithSkeletalMesh=true
  pixelDepthOffsetMode=PDOM_Legacy
  shadingModel=MSM_FromMaterialExpression
  usedShadingModels=Default Lit | Subsurface | Clear Coat | Subsurface Profile | Cloth
 }
 nodes [
  n1 = NamedRerouteUsage()
  n2 = NamedRerouteUsage()
  n3 = MaterialFunctionCall() {materialFunction=&MF_CharacterParent_2023_PreFX}
  MaterialFunctionCall(
   Material=StaticSwitchParameter(
    True=SetMaterialAttributes!(MaterialAttributes=@n1, Base Color=@n2) {attributeSetTypes [#1]}
    False=NamedRerouteUsage()
   ) {defaultValue=true, group=Digits, parameterName=EnableDigits, sortPriority=10}
   M=@n3.M
   Pre Skinned Local Position=@n3.M
   Distance Blend=@n3.M
   Fresnel=@n3.M
  ) {materialFunction=&MF_CharacterParent_2023_PostFX}
  NamedRerouteDeclaration(None=GetMaterialAttributes(None=NamedRerouteDeclaration(None=@n3.Result) {name=IN_Material, nodeColor=<rgba 1 0.0234 0 1>}) {attributeGetTypes [#1]}.Base Color) {name=IN_Diffuse, nodeColor=<rgba 0 0.2969 1 1>}
  n4 = Clamp(
   None=ScalarParameter() {defaultValue=7, group=Digits, parameterName=Number}
   Min=ScalarParameter() {defaultValue=1, group=Digits, parameterName=Number_Min, sortPriority=33}
   Max=ScalarParameter() {defaultValue=456, group=Digits, parameterName=Number_Max, sortPriority=34}
  ) {maxDefault=999}
  n5 = Constant() {r=2}
  n6 = Constant() {r=5}
  n7 = MaterialFunctionCall(UVs=TextureCoordinate() {coordinateIndex=1}, Texture Scale=Constant2Vector() {g=0.8, r=0.8}) {materialFunction=&ScaleUVsByCenter}
  n8 = Multiply(A=@n7.UVs, B=Constant2Vector() {g=1, r=3}) {constB=3}
  n9 = Saturate(None=@n8)
  n10 = Constant2Vector() {g=1, r=1.2}
  n11 = TextureSampleParameter2D(UVs=MaterialFunctionCall(
    "Animation  Phase (0-1)"=Divide(A=@n4) {constB=1000}
    Number of Rows=@n5
    Number of Columns=@n6
    UVs=MaterialFunctionCall(UVs=@n9, Texture Scale=@n10) {materialFunction=&ScaleUVsByCenter}.UVs
   ) {materialFunction=&FlipBook}.UVs) {
   group=Digits
   parameterName=Mask_Digits
   samplerType=SAMPLERTYPE_Masks
   texture=/BRCosmetics/Characters/Player/Male/Medium/Bodies/M_MED_TimberStakeDrift/Textures/T_TimberStakeDrift_Numbers
  }
  n12 = Fmod(A=@n4, B=Constant() {r=100})
  n13 = Subtract(A=@n8, B=Constant2Vector() {r=1})
  n14 = Saturate(None=@n13)
  n15 = TextureSampleParameter2D(UVs=MaterialFunctionCall(
    "Animation  Phase (0-1)"=Divide(A=@n12) {constB=100}
    Number of Rows=@n5
    Number of Columns=@n6
    UVs=MaterialFunctionCall(UVs=@n14, Texture Scale=@n10) {materialFunction=&ScaleUVsByCenter}.UVs
   ) {materialFunction=&FlipBook}.UVs) {
   group=Digits
   parameterName=Mask_Digits
   samplerType=SAMPLERTYPE_Masks
   texture=&T_TimberStakeDrift_Numbers
  }
  n16 = Fmod(A=@n4, B=Constant() {r=10})
  n17 = Saturate(None=Subtract(A=@n13, B=Constant2Vector() {r=1}))
  n18 = TextureSampleParameter2D(UVs=MaterialFunctionCall(
    "Animation  Phase (0-1)"=Divide(A=@n16) {constB=10}
    Number of Rows=@n5
    Number of Columns=@n6
    UVs=MaterialFunctionCall(UVs=@n17, Texture Scale=@n10) {materialFunction=&ScaleUVsByCenter}.UVs
   ) {materialFunction=&FlipBook}.UVs) {
   group=Digits
   parameterName=Mask_Digits
   samplerType=SAMPLERTYPE_Masks
   texture=&T_TimberStakeDrift_Numbers
  }
  n19 = Multiply(A=Max(A=@n11.R, B=Max(A=@n15.R, B=@n18.R)), B=NamedRerouteUsage())
  n20 = ScalarParameter() {defaultValue=1, group=Digits, parameterName=Digits_AlphaStrength, sortPriority=15}
  NamedRerouteDeclaration(None=LinearInterpolate(
    A=NamedRerouteUsage()
    B=VectorParameter() {group=Digits, parameterName=DigitsColor, defaultValue=<rgba 0.887 0.8341 0.6919 1>}.RGB
    Alpha=Multiply(A=@n19, B=@n20)
   )) {name=OUT_Diffuse, nodeColor=<rgba 1 0 0.7969 1>}
  NamedRerouteDeclaration(None=@n7.0-1 mask) {name=NumbersMask, nodeColor=<rgba 0 0.0625 1 1>}
 ]
}
