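# Physics asset for the M_MED_TimberStakeDrift male medium cosmetic body, defining bodies for its front flaps and zipper.
# Small shapes on pelvis and spine form a free-swinging chain, with flap base bodies locked to spine_05 and the zipper base free on spine_04.
# Box-shaped dynamic bones hang from these bases as two-segment front flap chains and a zipper, with limited swing and locked twist.
# Back and side flaps have only base bodies.
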PhysicsAsset /BRCosmetics/Characters/Player/Male/Medium/Bodies/M_MED_TimberStakeDrift/Meshes/M_MED_TimberStakeDrift_Physics
bodies {
 base_bk_flap_l [{shapeType=Box, extentX=1.01, extentY=1.01, extentZ=1.01, shapeName=base_bk_flap_l_box, rotation=<rot 0 -90 0>}]
 base_fr_flap_l [{shapeType=Capsule, length=0.5, radius=0.5, shapeName=base_fr_flap_l_capsule}]
 base_fr_flap_r [{shapeType=Capsule, length=0.5, radius=0.5, shapeName=base_fr_flap_r_capsule}]
 base_sd_flap_a_l [{shapeType=Capsule, length=0.1, radius=0.1, shapeName=base_sd_flap_a_l_capsule, rotation=<rot 0 -90 0>}]
 base_sd_flap_a_r [{shapeType=Capsule, length=0.1, radius=0.1, shapeName=base_sd_flap_a_r_capsule, rotation=<rot 0 -90 0>}]
 base_sd_flap_b_l [{shapeType=Capsule, length=0.1, radius=0.1, shapeName=base_sd_flap_b_l_capsule, rotation=<rot 0 -90 0>}]
 base_sd_flap_b_r [{shapeType=Capsule, length=0.1, radius=0.1, shapeName=base_sd_flap_b_r_capsule, rotation=<rot 0 -90 0>}]
 base_zipper [{shapeType=Capsule, length=0.1, radius=0.1, shapeName=base_zipper_capsule, rotation=<rot 0 -90 0>}]
 dyn_fr_flap_01_l [{
   shapeType=Box
   extentX=4.5825
   extentY=0.6611
   extentZ=3.6721
   shapeName=dyn_fr_flap_01_l_box
   center=<2.0568 0.8 0>
   rotation=<rot 0 -169.5306 -88.1098>
  }]
 dyn_fr_flap_01_r [{
   shapeType=Box
   extentX=5.5
   extentY=2
   extentZ=7.066
   shapeName=dyn_fr_flap_01_r_box
   center=<0.9086 0.7098 0>
   rotation=<rot 0 17.1225 -78.194>
  }]
 dyn_fr_flap_02_l [{
   shapeType=Box
   extentX=3.6627
   extentY=1.2093
   extentZ=6.9362
   shapeName=dyn_fr_flap_02_l_box
   center=<2.0338 -0.6 0.3>
   rotation=<rot 0 10.6321 -78.5281>
  }]
 dyn_fr_flap_02_r [{
   shapeType=Box
   extentX=2.9218
   extentY=1.7374
   extentZ=8.4002
   shapeName=dyn_fr_flap_02_r_box
   center=<1.5156 -0.24 0.13>
   rotation=<rot 0 -162.8461 -86.201>
  }]
 dyn_zipper [{shapeType=Box, extentX=1.8, extentY=0.7, extentZ=4.5, shapeName=dyn_zipper_box, center=<2 0 0>, rotation=<rot 0 90 90>}]
 pelvis [{shapeType=Capsule, length=1, radius=1, shapeName=pelvis_capsule}]
 spine_04 [{shapeType=Capsule, length=1, radius=1, shapeName=spine_04_capsule}]
 spine_05 [{shapeType=Capsule, length=1, radius=1, shapeName=spine_05_capsule}]
}
constraints [
 {bone1Name=dyn_zipper, bone2Name=base_zipper, swing1LimitDegrees=25, swing2LimitDegrees=25, twistLimitDegrees=45, twistMotion=Locked}
 {
  bone1Name=base_zipper
  bone2Name=spine_04
  swing1LimitDegrees=45
  swing1Motion=Free
  swing2LimitDegrees=45
  swing2Motion=Free
  twistLimitDegrees=45
  twistMotion=Free
 }
 {
  bone1Name=spine_04
  bone2Name=pelvis
  swing1LimitDegrees=45
  swing1Motion=Free
  swing2LimitDegrees=45
  swing2Motion=Free
  twistLimitDegrees=45
  twistMotion=Free
 }
 {
  bone1Name=base_fr_flap_r
  bone2Name=spine_05
  swing1LimitDegrees=45
  swing1Motion=Locked
  swing2LimitDegrees=45
  swing2Motion=Locked
  twistLimitDegrees=45
  twistMotion=Locked
 }
 {
  bone1Name=base_fr_flap_l
  bone2Name=spine_05
  swing1LimitDegrees=45
  swing1Motion=Locked
  swing2LimitDegrees=45
  swing2Motion=Locked
  twistLimitDegrees=45
  twistMotion=Locked
 }
 {bone1Name=dyn_fr_flap_01_l, bone2Name=base_fr_flap_l, swing1LimitDegrees=10, swing2LimitDegrees=8, twistLimitDegrees=45, twistMotion=Locked}
 {bone1Name=dyn_fr_flap_01_r, bone2Name=base_fr_flap_r, swing1LimitDegrees=10, swing2LimitDegrees=13, twistLimitDegrees=45, twistMotion=Locked}
 {bone1Name=dyn_fr_flap_02_r, bone2Name=dyn_fr_flap_01_r, swing1LimitDegrees=2, swing2LimitDegrees=6, twistLimitDegrees=45, twistMotion=Locked}
 {bone1Name=dyn_fr_flap_02_l, bone2Name=dyn_fr_flap_01_l, swing1LimitDegrees=3, swing2LimitDegrees=7, twistLimitDegrees=45, twistMotion=Locked}
 {
  bone1Name=base_bk_flap_l
  bone2Name=spine_05
  swing1LimitDegrees=45
  swing1Motion=Locked
  swing2LimitDegrees=45
  swing2Motion=Locked
  twistLimitDegrees=45
  twistMotion=Locked
 }
 {
  bone1Name=base_sd_flap_a_l
  bone2Name=spine_05
  swing1LimitDegrees=45
  swing1Motion=Locked
  swing2LimitDegrees=45
  swing2Motion=Locked
  twistLimitDegrees=45
  twistMotion=Locked
 }
 {
  bone1Name=base_sd_flap_a_r
  bone2Name=spine_05
  swing1LimitDegrees=45
  swing1Motion=Locked
  swing2LimitDegrees=45
  swing2Motion=Locked
  twistLimitDegrees=45
  twistMotion=Locked
 }
 {
  bone1Name=base_sd_flap_b_l
  bone2Name=spine_05
  swing1LimitDegrees=45
  swing1Motion=Locked
  swing2LimitDegrees=45
  swing2Motion=Locked
  twistLimitDegrees=45
  twistMotion=Locked
 }
 {
  bone1Name=base_sd_flap_b_r
  bone2Name=spine_05
  swing1LimitDegrees=45
  swing1Motion=Locked
  swing2LimitDegrees=45
  swing2Motion=Locked
  twistLimitDegrees=45
  twistMotion=Locked
 }
 {
  bone1Name=spine_05
  bone2Name=spine_04
  swing1LimitDegrees=45
  swing1Motion=Free
  swing2LimitDegrees=45
  swing2Motion=Free
  twistLimitDegrees=45
  twistMotion=Free
 }
]
