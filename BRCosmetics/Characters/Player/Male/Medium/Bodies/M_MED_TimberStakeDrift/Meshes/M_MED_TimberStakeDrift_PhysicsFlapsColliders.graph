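# Physics asset for the TimberStakeDrift male medium body cosmetic, providing bodies and joint limits for one back flap and three side flaps.
# Small base bodies lock each flap to spine_05, and each dynamic capsule hangs off its base with tight, per-flap swing and twist limits, some axes locked.
# Head and neck capsules form a locked chain on spine_05 as collider shapes, while spine_05, spine_04 and pelvis link through free constraints.
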PhysicsAsset /BRCosmetics/Characters/Player/Male/Medium/Bodies/M_MED_TimberStakeDrift/Meshes/M_MED_TimberStakeDrift_PhysicsFlapsColliders
bodies {
 base_bk_flap_l [{shapeType=Box, extentX=1.01, extentY=1.01, extentZ=1.01, shapeName=base_bk_flap_l_box, rotation=<rot 0 -90 0>}]
 base_sd_flap_a_l [{shapeType=Capsule, length=0.1, radius=0.1, shapeName=base_sd_flap_a_l_capsule, rotation=<rot 0 -90 0>}]
 base_sd_flap_a_r [{shapeType=Capsule, length=0.1, radius=0.1, shapeName=base_sd_flap_a_r_capsule, rotation=<rot 0 -90 0>}]
 base_sd_flap_b_l [{shapeType=Capsule, length=0.1, radius=0.1, shapeName=base_sd_flap_b_l_capsule, rotation=<rot 0 -90 0>}]
 dyn_bk_flap_l [{
   shapeType=Capsule
   length=6.8329
   radius=2.5799
   shapeName=dyn_bk_flap_l_capsule
   center=<3.799 0 -0.1978>
   rotation=<rot 90 -165.9638 104.0362>
  }]
 dyn_sd_flap_a_01_l [{
   shapeType=Capsule
   length=4.5927
   radius=2.4004
   shapeName=dyn_sd_flap_a_01_l_capsule
   center=<5.025 -0.116 -1.6724>
   rotation=<rot -3.3086 11.1492 -84.4211>
  }]
 dyn_sd_flap_a_01_r [{
   shapeType=Capsule
   length=3.158
   radius=2.129
   shapeName=dyn_sd_flap_a_01_r_capsule
   center=<3.5892 -0.591 0.049>
   rotation=<rot 0 -166.2249 -93.3917>
  }]
 dyn_sd_flap_b_01_l [{
   shapeType=Capsule
   length=6.4507
   radius=2.0499
   shapeName=dyn_sd_flap_b_01_l_capsule
   center=<3.3417 -0.8574 -0.027>
   rotation=<rot 0.3252 1.2523 -82.7734>
  }]
 head [
  {shapeType=Capsule, length=2.5, radius=4.75, shapeName=head_capsule, center=<1.7164 -1.1991 0>, rotation=<rot 90 0 0>}
  {shapeType=Capsule, length=1.5, radius=3.5, shapeName=head_capsule, center=<-2.517 6.8 -0.2327>}
  {shapeType=Capsule, length=2.5, radius=5, shapeName=head_capsule, center=<-2.0909 1.1 0>, rotation=<rot 90 180 90>}
 ]
 neck_01 [{shapeType=Capsule, length=4, radius=3.8, shapeName=neck_01_capsule, center=<1.3067 1.0602 0>, rotation=<rot 90 0 0>}]
 neck_02 [{shapeType=Capsule, length=5.5, radius=3.5, shapeName=neck_02_capsule, center=<0.712 0 0>, rotation=<rot 90 149.9667 154.2426>}]
 pelvis [{shapeType=Capsule, length=1, radius=1, shapeName=pelvis_capsule}]
 spine_04 [{shapeType=Capsule, length=1, radius=1, shapeName=spine_04_capsule}]
 spine_05 [{shapeType=Capsule, length=1, radius=1, shapeName=spine_05_capsule}]
}
constraints [
 {
  bone1Name=spine_04
  bone2Name=pelvis
  swing1LimitDegrees=45
  swing1Motion=Free
  swing2LimitDegrees=45
  swing2Motion=Free
  twistLimitDegrees=45
  twistMotion=Free
 }
 {
  bone1Name=neck_01
  bone2Name=spine_05
  swing1LimitDegrees=45
  swing1Motion=Locked
  swing2LimitDegrees=45
  swing2Motion=Locked
  twistLimitDegrees=45
  twistMotion=Locked
 }
 {
  bone1Name=base_bk_flap_l
  bone2Name=spine_05
  swing1LimitDegrees=45
  swing1Motion=Locked
  swing2LimitDegrees=45
  swing2Motion=Locked
  twistLimitDegrees=45
  twistMotion=Locked
 }
 {
  bone1Name=neck_02
  bone2Name=neck_01
  swing1LimitDegrees=45
  swing1Motion=Locked
  swing2LimitDegrees=45
  swing2Motion=Locked
  twistLimitDegrees=45
  twistMotion=Locked
 }
 {
  bone1Name=head
  bone2Name=neck_02
  swing1LimitDegrees=45
  swing1Motion=Locked
  swing2LimitDegrees=45
  swing2Motion=Locked
  twistLimitDegrees=45
  twistMotion=Locked
 }
 {bone1Name=dyn_bk_flap_l, bone2Name=base_bk_flap_l, swing1LimitDegrees=5, swing1Motion=Locked, swing2LimitDegrees=20, twistLimitDegrees=25}
 {
  bone1Name=dyn_sd_flap_a_01_l
  bone2Name=base_sd_flap_a_l
  swing1LimitDegrees=5
  swing2LimitDegrees=40
  twistLimitDegrees=45
  twistMotion=Locked
 }
 {
  bone1Name=dyn_sd_flap_a_01_r
  bone2Name=base_sd_flap_a_r
  swing1LimitDegrees=5
  swing2LimitDegrees=30
  twistLimitDegrees=45
  twistMotion=Locked
 }
 {bone1Name=dyn_sd_flap_b_01_l, bone2Name=base_sd_flap_b_l, swing1LimitDegrees=5, swing2LimitDegrees=50, twistLimitDegrees=15}
 {
  bone1Name=base_sd_flap_a_l
  bone2Name=spine_05
  swing1LimitDegrees=45
  swing1Motion=Locked
  swing2LimitDegrees=45
  swing2Motion=Locked
  twistLimitDegrees=45
  twistMotion=Locked
 }
 {
  bone1Name=base_sd_flap_a_r
  bone2Name=spine_05
  swing1LimitDegrees=45
  swing1Motion=Locked
  swing2LimitDegrees=45
  swing2Motion=Locked
  twistLimitDegrees=45
  twistMotion=Locked
 }
 {
  bone1Name=base_sd_flap_b_l
  bone2Name=spine_05
  swing1LimitDegrees=45
  swing1Motion=Locked
  swing2LimitDegrees=45
  swing2Motion=Locked
  twistLimitDegrees=45
  twistMotion=Locked
 }
 {
  bone1Name=spine_05
  bone2Name=spine_04
  swing1LimitDegrees=45
  swing1Motion=Free
  swing2LimitDegrees=45
  swing2Motion=Free
  twistLimitDegrees=45
  twistMotion=Free
 }
]
