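# Physics asset for the TimberStakeDrift male character's FaceAcc part, defining capsule bodies and constraints for front and back hair movement.
# Short dyn_hair chains extend from base_hair bones with limited swing and locked twist; the back-right c chain branches.
# Back roots mostly lock to base_hair_bk_area, which, like the front roots, joins the head with free rotation.
# Larger head, neck, upper spine, clavicle and upper-arm capsules form the body volume, linked by free-rotation constraints.
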PhysicsAsset /BRCosmetics/Characters/Player/Male/Medium/Bodies/M_MED_TimberStakeDrift/Meshes/Parts/M_MED_TimberStakeDrift_FaceAcc_Physics
bodies {
 base_hair_bk_01_a_l [{shapeType=Capsule, length=0.505, radius=0.505, shapeName=base_hair_bk_01_a_l_capsule, rotation=<rot 0 -90 0>}]
 base_hair_bk_01_a_r [{shapeType=Capsule, length=0.505, radius=0.505, shapeName=base_hair_bk_01_a_r_capsule, rotation=<rot 0 -90 0>}]
 base_hair_bk_01_b_l [{shapeType=Capsule, length=0.505, radius=0.505, shapeName=base_hair_bk_01_b_l_capsule, rotation=<rot 0 -90 0>}]
 base_hair_bk_01_b_r [{shapeType=Capsule, length=0.505, radius=0.505, shapeName=base_hair_bk_01_b_r_capsule, rotation=<rot 0 -90 0>}]
 base_hair_bk_01_c_l [{shapeType=Capsule, length=0.505, radius=0.505, shapeName=base_hair_bk_01_c_l_capsule, rotation=<rot 0 -90 0>}]
 base_hair_bk_01_c_r [{shapeType=Capsule, length=0.505, radius=0.505, shapeName=base_hair_bk_01_c_r_capsule, rotation=<rot 0 -90 0>}]
 base_hair_bk_01_d_l [{shapeType=Capsule, length=0.505, radius=0.505, shapeName=base_hair_bk_01_d_l_capsule, rotation=<rot 0 -90 0>}]
 base_hair_bk_01_m [{shapeType=Capsule, length=0.505, radius=0.505, shapeName=base_hair_bk_01_m_capsule, rotation=<rot 0 -90 0>}]
 base_hair_bk_area [{shapeType=Capsule, length=0.505, radius=0.505, shapeName=base_hair_bk_area_capsule, rotation=<rot 0 -90 0>}]
 base_hair_fr_a_l [{shapeType=Capsule, length=0.505, radius=0.505, shapeName=base_hair_fr_a_l_capsule, rotation=<rot 0 -90 0>}]
 base_hair_fr_a_r [{shapeType=Capsule, length=0.505, radius=0.505, shapeName=base_hair_fr_a_r_capsule, rotation=<rot 0 -90 0>}]
 base_hair_fr_b_l [{shapeType=Capsule, length=0.505, radius=0.505, shapeName=base_hair_fr_b_l_capsule, rotation=<rot 0 -90 0>}]
 base_hair_fr_b_r [{shapeType=Capsule, length=0.505, radius=0.505, shapeName=base_hair_fr_b_r_capsule, rotation=<rot 0 -90 0>}]
 base_hair_fr_c_r [{shapeType=Capsule, length=0.505, radius=0.505, shapeName=base_hair_fr_c_r_capsule, rotation=<rot 0 -90 0>}]
 clavicle_l [{shapeType=Capsule, length=8, radius=5.5, shapeName=clavicle_l_capsule, center=<7.5851 0 0>, rotation=<rot 90 0 0>}]
 clavicle_r [{shapeType=Capsule, length=8, radius=5.5, shapeName=clavicle_r_capsule, center=<-7.5851 0 0>, rotation=<rot 90 0 180>}]
 dyn_hair_bk_01_a_l [{shapeType=Capsule, length=1.6098, radius=1.039, shapeName=dyn_hair_bk_01_a_l_capsule, center=<0.825 0 -0.0987>, rotation=<rot 90 0 0>}]
 dyn_hair_bk_01_a_r [{shapeType=Capsule, length=0.505, radius=0.505, shapeName=dyn_hair_bk_01_a_r_capsule, center=<0.6534 0 0>, rotation=<rot 90 0 0>}]
 dyn_hair_bk_01_b_l [{shapeType=Capsule, length=0.505, radius=0.505, shapeName=dyn_hair_bk_01_b_l_capsule, center=<0.6534 0 0>, rotation=<rot 90 0 0>}]
 dyn_hair_bk_01_b_r [{shapeType=Capsule, length=1.2028, radius=1.3806, shapeName=dyn_hair_bk_01_b_r_capsule, center=<0.8306 0 0.4671>, rotation=<rot 90 0 0>}]
 dyn_hair_bk_01_c_l [{shapeType=Capsule, length=0.505, radius=0.505, shapeName=dyn_hair_bk_01_c_l_capsule, center=<0.6534 0 0>, rotation=<rot 90 0 0>}]
 dyn_hair_bk_01_c_r [{
   shapeType=Capsule
   length=1.9245
   radius=0.8248
   shapeName=dyn_hair_bk_01_c_r_capsule
   center=<0.7585 -0.2984 0.3882>
   rotation=<rot 72.2182 4.5846 -9.0295>
  }]
 dyn_hair_bk_01_d_l [{shapeType=Capsule, length=0.7897, radius=0.6489, shapeName=dyn_hair_bk_01_d_l_capsule, center=<0.6534 0 0>, rotation=<rot 90 0 0>}]
 dyn_hair_bk_01_m [{shapeType=Capsule, length=1.3899, radius=0.8276, shapeName=dyn_hair_bk_01_m_capsule, center=<0.6534 0 0>, rotation=<rot 90 0 0>}]
 dyn_hair_bk_02_a_l [{
   shapeType=Capsule
   length=1.1071
   radius=0.6355
   shapeName=dyn_hair_bk_02_a_l_capsule
   center=<1.3474 0 -0.1405>
   rotation=<rot 83.3938 179.9999 179.9999>
  }]
 dyn_hair_bk_02_a_r [{
   shapeType=Capsule
   length=1.3188
   radius=0.9212
   shapeName=dyn_hair_bk_02_a_r_capsule
   center=<0.6974 0 0.2509>
   rotation=<rot 80.1483 179.9999 180>
  }]
 dyn_hair_bk_02_b_l [{shapeType=Capsule, length=0.8226, radius=0.5, shapeName=dyn_hair_bk_02_b_l_capsule, center=<0.6534 0 0>, rotation=<rot 90 0 0>}]
 dyn_hair_bk_02_b_r [{shapeType=Capsule, length=2, radius=0.9, shapeName=dyn_hair_bk_02_b_r_capsule, center=<0.498 0 0.2527>, rotation=<rot 73.1449 180 180>}]
 dyn_hair_bk_02_c_l [{shapeType=Capsule, length=1.1273, radius=0.5396, shapeName=dyn_hair_bk_02_c_l_capsule, center=<0.6534 0 0>, rotation=<rot 90 0 0>}]
 dyn_hair_bk_02_c_r [{shapeType=Capsule, length=1.4567, radius=1.0016, shapeName=dyn_hair_bk_02_c_r_capsule, center=<0.4577 -0.0272 0.1765>, rotation=<rot 90 0 0>}]
 dyn_hair_bk_02_d_l [{shapeType=Capsule, length=0.9176, radius=0.6236, shapeName=dyn_hair_bk_02_d_l_capsule, center=<0.6534 0 0>, rotation=<rot 90 0 0>}]
 dyn_hair_bk_02_m [{shapeType=Capsule, length=1.1513, radius=0.9491, shapeName=dyn_hair_bk_02_m_capsule, center=<0.6534 0 0>, rotation=<rot 90 0 0>}]
 dyn_hair_bk_03_c_r [{shapeType=Capsule, length=0.6236, radius=0.4515, shapeName=dyn_hair_bk_03_c_r_capsule, center=<1.0426 0 -0.0838>, rotation=<rot 90 0 0>}]
 dyn_hair_bk_04_c_r [{shapeType=Capsule, length=0.5122, radius=0.3511, shapeName=dyn_hair_bk_04_c_r_capsule, center=<0.9847 0 0.0207>, rotation=<rot 90 0 0>}]
 dyn_hair_fr_a_01_l [{shapeType=Capsule, length=1, radius=1, shapeName=dyn_hair_fr_a_01_l_capsule, center=<0.5 0 0>, rotation=<rot 90 0 0>}]
 dyn_hair_fr_a_01_r [{shapeType=Capsule, length=1, radius=1, shapeName=dyn_hair_fr_a_01_r_capsule, center=<0.893 0 0>, rotation=<rot 90 0 0>}]
 dyn_hair_fr_a_02_l [{shapeType=Capsule, length=1, radius=1.3, shapeName=dyn_hair_fr_a_02_l_capsule, center=<1.0143 0 0>, rotation=<rot 90 0 0>}]
 dyn_hair_fr_a_02_r [{shapeType=Capsule, length=2.5, radius=1, shapeName=dyn_hair_fr_a_02_r_capsule, center=<1.3083 0.5 0>, rotation=<rot 70.047 36.3276 0.0002>}]
 dyn_hair_fr_a_03_l [{
   shapeType=Capsule
   length=2
   radius=1
   shapeName=dyn_hair_fr_a_03_l_capsule
   center=<1.1264 0.2 0.11>
   rotation=<rot 84.4271 143.6052 -179.9998>
  }]
 dyn_hair_fr_b_01_l [{
   shapeType=Capsule
   length=2
   radius=1.25
   shapeName=dyn_hair_fr_b_01_l_capsule
   center=<0.668 0 0>
   rotation=<rot 82.3586 -179.9999 -179.9999>
  }]
 dyn_hair_fr_b_01_r [{shapeType=Capsule, length=1, radius=1, shapeName=dyn_hair_fr_b_01_r_capsule, center=<0.4885 0 0.3056>, rotation=<rot 90 0 0>}]
 dyn_hair_fr_b_02_l [{
   shapeType=Capsule
   length=2.5
   radius=1
   shapeName=dyn_hair_fr_b_02_l_capsule
   center=<0.6265 -0.421 0.3909>
   rotation=<rot 83.8414 146.1031 -179.9999>
  }]
 dyn_hair_fr_b_02_r [{
   shapeType=Capsule
   length=2.5
   radius=0.7
   shapeName=dyn_hair_fr_b_02_r_capsule
   center=<0.9586 0.5109 0.2541>
   rotation=<rot 70.8872 165.9265 125.5575>
  }]
 dyn_hair_fr_c_01_r [{
   shapeType=Capsule
   length=1.5
   radius=1
   shapeName=dyn_hair_fr_c_01_r_capsule
   center=<0.7413 0 0.0759>
   rotation=<rot 84.4043 0.0008 0.0009>
  }]
 dyn_hair_fr_c_02_r [{shapeType=Capsule, length=2.5, radius=1, shapeName=dyn_hair_fr_c_02_r_capsule, center=<0.8618 0.6993 -0.4045>, rotation=<rot 90 0 -39.0586>}]
 head [{shapeType=Capsule, length=5, radius=5.8, shapeName=head_capsule, center=<3.3876 4.0797 0>, rotation=<rot 90 0 0>}]
 neck_01 [
  {shapeType=Capsule, length=9.5, radius=5, shapeName=neck_01_capsule, center=<2.0364 1.5819 0>}
  {shapeType=Capsule, length=4, radius=3.5, shapeName=neck_01_capsule, center=<0.6574 -2.2211 7.8597>, rotation=<rot 80.3983 56.6983 0>}
 ]
 neck_02 [{shapeType=Capsule, length=3.5, radius=7.5, shapeName=neck_02_capsule, center=<-3.1577 3.0141 0>, rotation=<rot 90 0 -16.077>}]
 spine_05 [
  {shapeType=Capsule, length=11.5, radius=10, shapeName=spine_05_capsule, center=<4.2989 -3.1137 0>, rotation=<rot 0 -90 0>}
  {shapeType=Capsule, length=11, radius=7, shapeName=spine_05_capsule, center=<9.2462 0 0>}
 ]
 upperarm_l [{shapeType=Capsule, length=12, radius=4.5, shapeName=upperarm_l_capsule, center=<4.9752 0 0>, rotation=<rot 90 0 0>}]
 upperarm_r [{shapeType=Capsule, length=12, radius=4.5, shapeName=upperarm_r_capsule, center=<-4.9752 0 0>, rotation=<rot 90 0 180>}]
}
constraints [
 {
  bone1Name=dyn_hair_bk_01_a_l
  bone2Name=base_hair_bk_01_a_l
  swing1LimitDegrees=15
  swing2LimitDegrees=35
  twistLimitDegrees=45
  twistMotion=Locked
 }
 {
  bone1Name=dyn_hair_bk_02_a_l
  bone2Name=dyn_hair_bk_01_a_l
  swing1LimitDegrees=20
  swing2LimitDegrees=25
  twistLimitDegrees=45
  twistMotion=Locked
 }
 {
  bone1Name=dyn_hair_bk_01_a_r
  bone2Name=base_hair_bk_01_a_r
  swing1LimitDegrees=10
  swing2LimitDegrees=32
  twistLimitDegrees=45
  twistMotion=Locked
 }
 {
  bone1Name=dyn_hair_bk_02_a_r
  bone2Name=dyn_hair_bk_01_a_r
  swing1LimitDegrees=15
  swing2LimitDegrees=35
  twistLimitDegrees=45
  twistMotion=Locked
 }
 {
  bone1Name=dyn_hair_bk_01_b_l
  bone2Name=base_hair_bk_01_b_l
  swing1LimitDegrees=10
  swing2LimitDegrees=20
  twistLimitDegrees=45
  twistMotion=Locked
 }
 {
  bone1Name=dyn_hair_bk_02_b_l
  bone2Name=dyn_hair_bk_01_b_l
  swing1LimitDegrees=10
  swing2LimitDegrees=20
  twistLimitDegrees=45
  twistMotion=Locked
 }
 {
  bone1Name=dyn_hair_bk_01_b_r
  bone2Name=base_hair_bk_01_b_r
  swing1LimitDegrees=20
  swing2LimitDegrees=40
  twistLimitDegrees=45
  twistMotion=Locked
 }
 {
  bone1Name=dyn_hair_bk_02_b_r
  bone2Name=dyn_hair_bk_01_b_r
  swing1LimitDegrees=15
  swing2LimitDegrees=30
  twistLimitDegrees=45
  twistMotion=Locked
 }
 {
  bone1Name=dyn_hair_bk_01_c_l
  bone2Name=base_hair_bk_01_c_l
  swing1LimitDegrees=10
  swing2LimitDegrees=25
  twistLimitDegrees=45
  twistMotion=Locked
 }
 {
  bone1Name=dyn_hair_bk_02_c_l
  bone2Name=dyn_hair_bk_01_c_l
  swing1LimitDegrees=10
  swing2LimitDegrees=35
  twistLimitDegrees=45
  twistMotion=Locked
 }
 {
  bone1Name=dyn_hair_bk_01_c_r
  bone2Name=base_hair_bk_01_c_r
  swing1LimitDegrees=20
  swing2LimitDegrees=30
  twistLimitDegrees=45
  twistMotion=Locked
 }
 {
  bone1Name=dyn_hair_bk_02_c_r
  bone2Name=dyn_hair_bk_01_c_r
  swing1LimitDegrees=20
  swing2LimitDegrees=15
  twistLimitDegrees=45
  twistMotion=Locked
 }
 {
  bone1Name=dyn_hair_bk_03_c_r
  bone2Name=dyn_hair_bk_01_c_r
  swing1LimitDegrees=20
  swing2LimitDegrees=20
  twistLimitDegrees=45
  twistMotion=Locked
 }
 {
  bone1Name=dyn_hair_bk_04_c_r
  bone2Name=dyn_hair_bk_01_c_r
  swing1LimitDegrees=20
  swing2LimitDegrees=20
  twistLimitDegrees=45
  twistMotion=Locked
 }
 {
  bone1Name=dyn_hair_bk_01_d_l
  bone2Name=base_hair_bk_01_d_l
  swing1LimitDegrees=10
  swing2LimitDegrees=25
  twistLimitDegrees=45
  twistMotion=Locked
 }
 {
  bone1Name=dyn_hair_bk_02_d_l
  bone2Name=dyn_hair_bk_01_d_l
  swing1LimitDegrees=10
  swing2LimitDegrees=25
  twistLimitDegrees=45
  twistMotion=Locked
 }
 {
  bone1Name=dyn_hair_bk_01_m
  bone2Name=base_hair_bk_01_m
  swing1LimitDegrees=10
  swing2LimitDegrees=35
  twistLimitDegrees=45
  twistMotion=Locked
 }
 {
  bone1Name=dyn_hair_bk_02_m
  bone2Name=dyn_hair_bk_01_m
  swing1LimitDegrees=10
  swing2LimitDegrees=35
  twistLimitDegrees=45
  twistMotion=Locked
 }
 {
  bone1Name=dyn_hair_fr_a_01_l
  bone2Name=base_hair_fr_a_l
  swing1LimitDegrees=5
  swing2LimitDegrees=10
  twistLimitDegrees=45
  twistMotion=Locked
 }
 {
  bone1Name=dyn_hair_fr_a_02_l
  bone2Name=dyn_hair_fr_a_01_l
  swing1LimitDegrees=7
  swing2LimitDegrees=15
  twistLimitDegrees=45
  twistMotion=Locked
 }
 {
  bone1Name=dyn_hair_fr_a_03_l
  bone2Name=dyn_hair_fr_a_02_l
  swing1LimitDegrees=6
  swing2LimitDegrees=10
  twistLimitDegrees=45
  twistMotion=Locked
 }
 {
  bone1Name=dyn_hair_fr_a_01_r
  bone2Name=base_hair_fr_a_r
  swing1LimitDegrees=13
  swing2LimitDegrees=2.5
  twistLimitDegrees=45
  twistMotion=Locked
 }
 {
  bone1Name=dyn_hair_fr_a_02_r
  bone2Name=dyn_hair_fr_a_01_r
  swing1LimitDegrees=13
  swing2LimitDegrees=3
  twistLimitDegrees=45
  twistMotion=Locked
 }
 {
  bone1Name=dyn_hair_fr_b_01_l
  bone2Name=base_hair_fr_b_l
  swing1LimitDegrees=10
  swing2LimitDegrees=10
  twistLimitDegrees=45
  twistMotion=Locked
 }
 {
  bone1Name=dyn_hair_fr_b_02_l
  bone2Name=dyn_hair_fr_b_01_l
  swing1LimitDegrees=10
  swing2LimitDegrees=10
  twistLimitDegrees=45
  twistMotion=Locked
 }
 {
  bone1Name=dyn_hair_fr_b_01_r
  bone2Name=base_hair_fr_b_r
  swing1LimitDegrees=10
  swing2LimitDegrees=10
  twistLimitDegrees=45
  twistMotion=Locked
 }
 {
  bone1Name=dyn_hair_fr_b_02_r
  bone2Name=dyn_hair_fr_b_01_r
  swing1LimitDegrees=10
  swing2LimitDegrees=3
  twistLimitDegrees=45
  twistMotion=Locked
 }
 {
  bone1Name=dyn_hair_fr_c_01_r
  bone2Name=base_hair_fr_c_r
  swing1LimitDegrees=10
  swing2LimitDegrees=4
  twistLimitDegrees=45
  twistMotion=Locked
 }
 {
  bone1Name=dyn_hair_fr_c_02_r
  bone2Name=dyn_hair_fr_c_01_r
  swing1LimitDegrees=10
  swing2LimitDegrees=3
  twistLimitDegrees=45
  twistMotion=Locked
 }
 {
  bone1Name=base_hair_fr_c_r
  bone2Name=head
  swing1LimitDegrees=45
  swing1Motion=Free
  swing2LimitDegrees=45
  swing2Motion=Free
  twistLimitDegrees=45
  twistMotion=Free
 }
 {
  bone1Name=base_hair_fr_b_r
  bone2Name=head
  swing1LimitDegrees=45
  swing1Motion=Free
  swing2LimitDegrees=45
  swing2Motion=Free
  twistLimitDegrees=45
  twistMotion=Free
 }
 {
  bone1Name=base_hair_fr_a_r
  bone2Name=head
  swing1LimitDegrees=45
  swing1Motion=Free
  swing2LimitDegrees=45
  swing2Motion=Free
  twistLimitDegrees=45
  twistMotion=Free
 }
 {
  bone1Name=base_hair_fr_b_l
  bone2Name=head
  swing1LimitDegrees=45
  swing1Motion=Free
  swing2LimitDegrees=45
  swing2Motion=Free
  twistLimitDegrees=45
  twistMotion=Free
 }
 {
  bone1Name=base_hair_fr_a_l
  bone2Name=head
  swing1LimitDegrees=45
  swing1Motion=Free
  swing2LimitDegrees=45
  swing2Motion=Free
  twistLimitDegrees=45
  twistMotion=Free
 }
 {
  bone1Name=base_hair_bk_01_d_l
  bone2Name=base_hair_bk_area
  swing1LimitDegrees=45
  swing1Motion=Locked
  swing2LimitDegrees=45
  swing2Motion=Locked
  twistLimitDegrees=45
  twistMotion=Locked
 }
 {
  bone1Name=base_hair_bk_01_c_l
  bone2Name=base_hair_bk_area
  swing1LimitDegrees=45
  swing1Motion=Locked
  swing2LimitDegrees=45
  swing2Motion=Locked
  twistLimitDegrees=45
  twistMotion=Locked
 }
 {
  bone1Name=base_hair_bk_01_b_l
  bone2Name=base_hair_bk_area
  swing1LimitDegrees=45
  swing1Motion=Locked
  swing2LimitDegrees=45
  swing2Motion=Locked
  twistLimitDegrees=45
  twistMotion=Locked
 }
 {
  bone1Name=base_hair_bk_01_a_l
  bone2Name=base_hair_bk_area
  swing1LimitDegrees=45
  swing1Motion=Locked
  swing2LimitDegrees=45
  swing2Motion=Locked
  twistLimitDegrees=45
  twistMotion=Locked
 }
 {
  bone1Name=base_hair_bk_01_m
  bone2Name=base_hair_bk_area
  swing1LimitDegrees=45
  swing1Motion=Free
  swing2LimitDegrees=45
  swing2Motion=Free
  twistLimitDegrees=45
  twistMotion=Free
 }
 {
  bone1Name=base_hair_bk_01_a_r
  bone2Name=base_hair_bk_area
  swing1LimitDegrees=45
  swing1Motion=Locked
  swing2LimitDegrees=45
  swing2Motion=Locked
  twistLimitDegrees=45
  twistMotion=Locked
 }
 {
  bone1Name=base_hair_bk_01_b_r
  bone2Name=base_hair_bk_area
  swing1LimitDegrees=45
  swing1Motion=Locked
  swing2LimitDegrees=45
  swing2Motion=Locked
  twistLimitDegrees=45
  twistMotion=Locked
 }
 {
  bone1Name=base_hair_bk_01_c_r
  bone2Name=base_hair_bk_area
  swing1LimitDegrees=45
  swing1Motion=Locked
  swing2LimitDegrees=45
  swing2Motion=Locked
  twistLimitDegrees=45
  twistMotion=Locked
 }
 {
  bone1Name=head
  bone2Name=neck_02
  swing1LimitDegrees=45
  swing1Motion=Free
  swing2LimitDegrees=45
  swing2Motion=Free
  twistLimitDegrees=45
  twistMotion=Free
 }
 {
  bone1Name=upperarm_l
  bone2Name=clavicle_l
  swing1LimitDegrees=45
  swing1Motion=Free
  swing2LimitDegrees=45
  swing2Motion=Free
  twistLimitDegrees=45
  twistMotion=Free
 }
 {
  bone1Name=upperarm_r
  bone2Name=clavicle_r
  swing1LimitDegrees=45
  swing1Motion=Free
  swing2LimitDegrees=45
  swing2Motion=Free
  twistLimitDegrees=45
  twistMotion=Free
 }
 {
  bone1Name=neck_02
  bone2Name=neck_01
  swing1LimitDegrees=45
  swing1Motion=Free
  swing2LimitDegrees=45
  swing2Motion=Free
  twistLimitDegrees=45
  twistMotion=Free
 }
 {
  bone1Name=neck_01
  bone2Name=spine_05
  swing1LimitDegrees=45
  swing1Motion=Free
  swing2LimitDegrees=45
  swing2Motion=Free
  twistLimitDegrees=45
  twistMotion=Free
 }
 {
  bone1Name=clavicle_l
  bone2Name=spine_05
  swing1LimitDegrees=45
  swing1Motion=Free
  swing2LimitDegrees=45
  swing2Motion=Free
  twistLimitDegrees=45
  twistMotion=Free
 }
 {
  bone1Name=clavicle_r
  bone2Name=spine_05
  swing1LimitDegrees=45
  swing1Motion=Free
  swing2LimitDegrees=45
  swing2Motion=Free
  twistLimitDegrees=45
  twistMotion=Free
 }
 {
  bone1Name=base_hair_bk_area
  bone2Name=head
  swing1LimitDegrees=45
  swing1Motion=Free
  swing2LimitDegrees=45
  swing2Motion=Free
  twistLimitDegrees=45
  twistMotion=Free
 }
]
